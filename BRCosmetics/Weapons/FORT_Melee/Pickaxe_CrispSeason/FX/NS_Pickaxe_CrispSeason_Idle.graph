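# Niagara idle effect for the CrispSeason pickaxe, likely paint drips given its brush and paint parameters.
# A short-lived mesh particle stretches and swings from the emitter on a pendulum constraint, oriented toward its origin, feeding origin and position to dynamic material parameters.
# A delayed sprite burst inherits some velocity, then falls, shrinks and fades.
# System-level time-of-day color tints both; a system speed alpha scales mesh size and sprite opacity.
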NiagaraSystem /BRCosmetics/Weapons/FORT_Melee/Pickaxe_CrispSeason/FX/NS_Pickaxe_CrispSeason_Idle
dependencies {
 usedDataInterfaces [
  &NiagaraDataInterfaceVectorCurve
  &NiagaraDataInterfaceCurve
  &NiagaraDataInterfaceVector2DCurve
 ]
 usedDynamicInputs [
  &SelectLinearColorFromArray
  &FNTimeOfDayColor
  &Divide_Float
  &VectorLength
 ]
 usedModules [
  &SystemState
  &self:SystemSpawnScript.SystemScriptSource.SystemScriptGraph.NiagaraNodeAssignment_6.SetVariables_#1
  &self:SystemSpawnScript.SystemScriptSource.SystemScriptGraph.NiagaraNodeAssignment_4.SetVariables_#2
  &self:SystemSpawnScript.SystemScriptSource.SystemScriptGraph.NiagaraNodeAssignment_3.SetVariables_#3
 ]
}
emitters {
 Particles_Swing002 {
  emitterUpdateScript [
   &EmitterState()
   &SpawnBurst_Instantaneous(Use Spawn Probability=false, UseLoopCountLimit=false)
  ]
  particleSpawnScript [&InitializeParticle(
    Color=$System.TOD Color
    Lifetime=0.4
    Mesh Scale=<0.1 0.1 0.05>
    Mesh Scale Mode=::NewEnumerator4
    Sprite Rotation Mode=::NewEnumerator2
    UsePositionOffset=false
    Write Material Random=true
   )]
  particleUpdateScript [
   &ParticleState()
   /SetVariables(Particles.MeshOrigin:NiagaraPosition=$Engine.Emitter.SimulationPosition) {isSetParameters=true}
   &Drag(Drag=5, Use Linear Drag=true, Use Rotational Drag=false)
   &GravityForce(Gravity=<0 0 -900>)
   &PendulumSetup(
    Module.Enable Spring Driver=false
    Pendulum Length=10
    Pendulum Pivot:Vector3f=$Particles.MeshOrigin
    "Spring Driven Constraint (Tightness)"=3
   )
   &SolveForcesAndVelocity(Clamp Velocity=false, Limit Acceleration=false)
   &PendulumConstraint()
   &CalculateAccurateVelocity()
   /SetVariables(Particles.MeshFacing:Vector3f=&NormalizeVector(
     Vector To Normalize=&Subtract_Vector(A=$Particles.MeshOrigin, B=Linked() {linkedVariable {name=Particles.Position, type=NiagaraPosition}})
     Safe Normalize=true
     Safe Fallback Threshold=0
     Safe Fallback Vector=<0 0 0.001>
    )) {isSetParameters=true}
   &ScaleMeshSize(Scale Factor=&VectorFromCurve(VectorCurve=DataInterface() {propertyValues {
       XCurve {keys [<at 0 0>,<at 0.1 1>]}
       YCurve {keys [<at 0 0>,<at 0.1 1>]}
       ZCurve {keys [
         {arriveTangent=0.2017, interpMode=RCIM_Cubic, leaveTangent=0.2017, tangentMode=RCTM_User, time=0, value=0}
         {arriveTangent=1.2, interpMode=RCIM_Cubic, leaveTangent=1.2, time=0.225, value=0.4856}
         {arriveTangent=1.0713, interpMode=RCIM_Cubic, leaveTangent=-3.2278, tangentMode=RCTM_Break, time=0.75, value=0.9}
         {arriveTangent=-0.4385, interpMode=RCIM_Cubic, leaveTangent=-0.4385, tangentMode=RCTM_User, time=1, value=0.4}
        ]}
      }}))
   &ScaleMeshSize(Scale Factor=&VectorFromFloat(Value=$System.SystemSpeedAlpha)) {module=ScaleMeshSize001}
   /Niagara/Modules/Update/Orientation/V2/OrientMeshToVector(
    Look At Direction:Vector3f=&Multiply_VectorByFloat(Vector=$Particles.MeshFacing, Float=1)
    Override Quat=false
    Reference Axis Vector:Vector3f=<0 0 1>
    UseOptionalSideAxis=false
    UseOptionalUpAxis=false
   )
   &DynamicMaterialParameters(
    Index 2 Param 1=HlslExpression() {hlslExpression=Particles.MeshOrigin.x}
    Index 2 Param 2=HlslExpression() {hlslExpression=Particles.MeshOrigin.y}
    Index 2 Param 3=HlslExpression() {hlslExpression=Particles.MeshOrigin.z}
    Index 2 Param 4=1.5
    Index 3 Param 1=&MakeFloatFromVector(Vector=Linked() {linkedVariable {name=Particles.Position, type=NiagaraPosition}}, Channel=::NewEnumerator0)
    Index 3 Param 2=&MakeFloatFromVector(Vector=Linked() {linkedVariable {name=Particles.Position, type=NiagaraPosition}}, Channel=::NewEnumerator1)
    Index 3 Param 3=&MakeFloatFromVector(Vector=Linked() {linkedVariable {name=Particles.Position, type=NiagaraPosition}}, Channel=::NewEnumerator2)
    Index 3 Param 4=0.2
    Module.Index2Float0Write=true
    Module.Index2Float1Write=true
    Module.Index2Float2Write=true
    Module.Index2Float3Write=true
    Module.Index3Float0Write=true
    Module.Index3Float1Write=true
    Module.Index3Float2Write=true
    Module.Index3Float3Write=true
    Write Parameter Index 0=::NewEnumerator0
    Write Parameter Index 2=::NewEnumerator1
    Write Parameter Index 3=::NewEnumerator1
   )
  ]
  rendererClasses [&NiagaraMeshRendererProperties]
  renderers [{rendererClass=&NiagaraMeshRendererProperties, rendererIndex=0}]
 }
 Particles_Swing004 {
  emitterUpdateScript [
   &EmitterState()
   &SpawnBurst_Instantaneous(Spawn Time=0.3, Use Spawn Probability=false, UseLoopCountLimit=false)
  ]
  particleSpawnScript [
   &InitializeParticle(
    Color=$System.TOD Color
    Lifetime Max=0.6
    Lifetime Min=0.3
    Lifetime Mode=::NewEnumerator1
    Sprite Size Mode=::NewEnumerator1
    Uniform Sprite Size Max=2.4
    Uniform Sprite Size Min=2
    UsePositionOffset=false
    Write Material Random=true
   )
   /SetVariables(Particles.Color=&MultiplyLinearColor(Linear Color=$Particles.Color, Scale Factor=&MakeVector4(X=1, Y=1, Z=1, W=$System.SystemSpeedAlpha))) {isSetParameters=true}
   &InheritVelocity(Inherited Velocity Amount Scale=<0.25 0.25 0.25>, Inherited Velocity Speed Limit=20, Speed Limit=false, UseSpeedThreshold=false)
  ]
  particleUpdateScript [
   &ParticleState()
   &Drag(Drag=&RandomRangeFloat(Minimum=1, Maximum=7), Use Linear Drag=true, Use Rotational Drag=false)
   &GravityForce(Gravity=<0 0 -500>)
   &ScaleColor(
    Scale Alpha=&Multiply_Float(
     A=&FloatFromCurve(FloatCurve=DataInterface() {propertyValues {Curve {
         defaultValue=0
         keys [{interpMode=RCIM_Cubic, tangentMode=RCTM_User, time=0.4403, value=1}, {interpMode=RCIM_Cubic, tangentMode=RCTM_User, time=1, value=0}]
        }}})
     B=$System.SystemSpeedAlpha
    )
    ScaleA=true
    ScaleRGB=true
   )
   &ScaleSpriteSize(Uniform Curve Sprite Scale=DataInterface() {propertyValues {Curve {keys [
        {arriveTangent=2.753, interpMode=RCIM_Cubic, leaveTangent=2.753, tangentMode=RCTM_User, time=0, value=0.3494}
        {arriveTangent=0.143, interpMode=RCIM_Cubic, leaveTangent=0.143, tangentMode=RCTM_User, time=0.3012, value=0.9559}
        {interpMode=RCIM_Cubic, tangentMode=RCTM_User, time=1, value=0.1882}
       ]}}}) {module=ScaleSpriteSize001}
   &SolveForcesAndVelocity(Clamp Velocity=false, Limit Acceleration=false)
  ]
  rendererClasses [&NiagaraSpriteRendererProperties]
  renderers [{rendererClass=&NiagaraSpriteRendererProperties}]
 }
}
summary {
 systemName=NS_Pickaxe_CrispSeason_Idle
 emitters [
  {emitterName=Particles_Swing002, rendererClasses [&NiagaraMeshRendererProperties]}
  {emitterName=Particles_Swing004, rendererClasses [&NiagaraSpriteRendererProperties]}
 ]
 userVariables [
  User.bIsFrontEnd {type=NiagaraBool}
  User.bIsFrontEndPreview {type=NiagaraBool}
  User.BrushColorArray {
   type=NiagaraDataInterfaceArrayColor
   description=""
   defaultValue {
    struct=NiagaraExt_VariableValue_DataInterface
    value {
     dataInterface=&self:NiagaraDataInterfaceArrayColor_0
     dataInterfaceClass=&NiagaraDataInterfaceArrayColor
    }
   }
  }
  User.Lifetime {type=NiagaraFloat, defaultValue=1.5, description=""}
  User.PaintID {type=NiagaraInt32, defaultValue=5}
  User.PhysicalSurface {type=NiagaraInt32}
  User.RibbonIDOffset {type=NiagaraInt32}
  User.RibbonWidth {type=NiagaraFloat, defaultValue=8, description=""}
  User.VariantID {type=NiagaraFloat, defaultValue=0, description=""}
 ]
}
