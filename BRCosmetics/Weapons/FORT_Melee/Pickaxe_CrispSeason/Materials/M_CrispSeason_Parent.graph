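# Parent material for the CrispSeason pickaxe, built on the 2023 character parent material functions.
# The PreFX function output has its Base Color lerped toward a PaintColor parameter, defaulting to orange, in the Reactivity group, masked by the red channel of the Tech Art Mask.
# The adjusted attributes then pass into the PostFX function.
# It selects shading model from the graph and supports skeletal, morph target, mesh deformer and Niagara mesh usage.
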Material /BRCosmetics/Weapons/FORT_Melee/Pickaxe_CrispSeason/Materials/M_CrispSeason_Parent
graph {
 properties {
  bUseMaterialAttributes=true
  bUsedWithMeshDeformer=true
  bUsedWithMorphTargets=true
  bUsedWithNiagaraMeshParticles=true
  bUsedWithSkeletalMesh=true
  pixelDepthOffsetMode=PDOM_Legacy
  shadingModel=MSM_FromMaterialExpression
  usedShadingModels=Default Lit | Subsurface | Clear Coat | Subsurface Profile | Cloth
 }
 nodes [
  n1 = MaterialFunctionCall() {materialFunction=&MF_CharacterParent_2023_PreFX}
  n2 = GetMaterialAttributes(None=@n1.Result) {attributeGetTypes [#1]}
  n3 = VectorParameter() {group=03_Reactivity, parameterName=PaintColor, defaultValue=<rgba 0.8388 0.162 0 1>}
  n4 = ComponentMask(None=@n1.Tech Art Mask) {r=true}
  MaterialFunctionCall(
   Material=SetMaterialAttributes!(MaterialAttributes=@n2.MaterialAttributes, Base Color=LinearInterpolate(A=@n2.Base Color, B=@n3.RGB, Alpha=@n4)) {attributeSetTypes [#1]}
   M=@n1.M
   Pre Skinned Local Position=@n1.M
   Distance Blend=@n1.M
   Fresnel=@n1.M
  ) {materialFunction=&MF_CharacterParent_2023_PostFX}
 ]
}
